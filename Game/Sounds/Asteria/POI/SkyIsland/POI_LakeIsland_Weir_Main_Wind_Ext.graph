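# Ambient SoundCue for the exterior wind at the Lake Island weir POI on Asteria's SkyIsland, going by its name.
# A single looping WavePlayer feeds the output, reusing the Apollo Cosmos biome WateringHole interior loop rather than a dedicated wind recording.
# It plays at reduced volume through the outdoor ambient sound class and Athena ambient attenuation, with a pre-effect bus send entry.
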SoundCue /Game/Sounds/Asteria/POI/SkyIsland/POI_LakeIsland_Weir_Main_Wind_Ext
root=WavePlayer_2
nodes [{id=WavePlayer_2, properties {bLooping=true, soundWaveAssetPtr=/Game/Sounds/Apollo/Biomes/Cosmos/AMB_Cosmos_WateringHole_Int_Loop_01}}]
properties {
 attenuationSettings=&Fort_Ambient_Athena_Att
 maxDistance=2097152
 soundClassObject=&Fort_Ambient_Outdoors_Athena
 volumeMultiplier=0.301
 preEffectBusSends [{}]
}
